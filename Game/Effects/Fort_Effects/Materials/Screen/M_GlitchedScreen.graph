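# Unlit, two-sided additive material for Niagara sprites that draws a glitched-screen effect tinted by particle color.
# A noise texture pans in stepped time in two layers, each sampled with chromatic offsets and multiplied, with a static switch choosing RGB split or two-color blend.
# An edge-shaped mask combines with per-cell grid noise, seeded by dynamic parameters and flickering by sine, and particle alpha.
# Several branches, including fire-flicker curve and hash noise, feed nothing visible.
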Material /Game/Effects/Fort_Effects/Materials/Screen/M_GlitchedScreen
graph {
 properties {
  bAutomaticallySetUsageInEditor=false
  bUsedWithNiagaraSprites=true
  blendMode=BLEND_Additive
  shadingModel=MSM_Unlit
  twoSided=true
  usedShadingModels=Unlit
 }
 nodes [
  n1 = Frac(None=TextureCoordinate() {uTiling=5, vTiling=5})
  n2 = ComponentMask(None=@n1) {r=true}
  Abs(None=Multiply(A=@n2, B=ComponentMask(None=@n1) {g=true}))
  n3 = TextureCoordinate()
  Abs(None=MaterialFunctionCall(UV Channel=@n3) {materialFunction=&LinearGradient}.VGradient)
  n4 = Add(A=TextureCoordinate() {uTiling=10, vTiling=10}) {constB=0.5}
  n5 = OneMinus(None=Sine(None=ComponentMask(None=@n4) {g=true, r=true}) {period=2})
  n6 = Saturate(None=Multiply(A=Power(Base=@n5) {constExponent=4}) {constB=3})
  Add(A=ComponentMask(None=@n6) {r=true}, B=ComponentMask(None=@n6) {g=true})
  n7 = TextureCoordinate() {vTiling=0.2}
  n8 = Frac(None=ScalarParameter() {defaultValue=0.5, parameterName=RandomLightVal})
  Add(A=@n7, B=@n8)
  n9 = Time()
  n10 = ScalarParameter() {defaultValue=0.5, parameterName=Speed}
  n11 = Constant() {r=5}
  n12 = Multiply(A=Frac(None=Multiply(A=@n9, B=@n10)), B=@n11)
  n13 = Panner(Coordinate=Multiply(A=TextureCoordinate()), Time=Divide(A=Floor(None=@n12), B=@n11)) {speedX=2}
  Clamp(
   None=Desaturation(None=TextureSample(UVs=@n13) {texture=/Game/Athena/UI/Challenges/Art/T_UI_DataChip_Noise}.RGB)
   Min=ScalarParameter() {defaultValue=0.45, parameterName=Minimum Value}
  )
  n14 = Time()
  n15 = LinearInterpolate(
   A=ScalarParameter() {defaultValue=0.75, parameterName=FlickerSpeed}
   B=ScalarParameter() {defaultValue=0.133, parameterName=FlickerSpeed2}
   Alpha=@n8
  )
  n16 = CurveAtlasRowParameter(CurveTime=Frac(None=Multiply(A=@n14, B=@n15))) {
   atlas=/Game/Environments/Helios/Props/Brimstone/Brimstone_Brazier_A/Blueprints/CurveAtlas_FN_LightFunctions
   curve=/Game/Blueprints/Lighting/LightFunctions/Curve_FireFlickers
   parameterName=FireFlicker
  }
  n17 = Constant() {r=0.5}
  n18 = ScalarParameter() {defaultValue=0.25, parameterName=Min}
  n19 = TextureSample(UVs=Panner(Coordinate=@n7) {speedY=-0.33}) {texture=/Game/Global/Textures/Default/Noise/3xPerlin_M}
  n20 = LinearInterpolate(A=@n18, Alpha=@n19.R) {constA=0.66}
  n21 = SphereMask(A=ComponentMask(None=@n3) {g=true}, B=Constant() {r=0.5}) {attenuationRadius=0.3, hardnessPercent=1}
  Clamp(None=Multiply(A=@n16.R, B=LinearInterpolate(A=@n17, B=@n20, Alpha=@n21)), Min=@n18)
  DynamicParameter() {paramNames [InnerEdgePower, TimeSpeed, RandomSeed, ChromaticLerp]}
  n22 = ScalarParameter() {defaultValue=20, parameterName=Tiling}
  n23 = Multiply(A=TextureCoordinate(), B=@n22)
  n24 = DotProduct(A=Divide(A=Floor(None=@n23), B=@n22), B=Constant2Vector() {g=78.233, r=12.9898})
  Frac(None=Multiply(A=Sine(None=@n24), B=Constant() {r=100000}))
  MaterialFunctionCall() {materialFunction=&LocalPosition}
  n25 = ParticleColor()
  n26 = TextureCoordinate()
  n27 = LinearInterpolate(
   A=Constant2Vector()
   B=ComponentMask(None=VectorParameter() {parameterName=ChromaOffset, defaultValue=<rgba 0.01 0.01 0 0>}.RGB) {g=true, r=true}
   Alpha=ScalarParameter() {defaultValue=1, parameterName=ChromaticIntensity}
  )
  n28 = Multiply(A=DynamicParameter() {paramNames [InnerEdgePower, TimeSpeed, RandomSeed, ChromaticLerp]}.TimeSpeed, B=Time())
  n29 = ScalarParameter() {defaultValue=20, parameterName=Steps}
  n30 = Divide(A=Ceil(None=Multiply(A=@n28, B=@n29)), B=@n29)
  n31 = VectorParameter() {parameterName=BlocksPanSpeed, defaultValue=<rgba 1.8 0.1 -0.8 -0.62>}
  n32 = AppendVector(A=@n31.R, B=@n31.R)
  n33 = TextureSample(UVs=Panner(Coordinate=Add(A=@n26, B=@n27), Time=@n30, Speed=@n32) {bFractionalPart=true}) {texture=&T_UI_DataChip_Noise}
  n34 = TextureSample(UVs=Panner(Coordinate=@n26, Time=@n30, Speed=@n32) {bFractionalPart=true}) {texture=&T_UI_DataChip_Noise}
  n35 = Multiply(A=@n27) {constB=-1}
  n36 = TextureSample(UVs=Panner(Coordinate=Add(A=@n35, B=@n26), Time=@n30, Speed=@n32) {bFractionalPart=true}) {texture=&T_UI_DataChip_Noise}
  n37 = AppendVector(A=@n31.B, B=@n31.B)
  n38 = TextureSample(UVs=Panner(Coordinate=Add(A=@n27, B=@n26), Time=@n30, Speed=@n37) {bFractionalPart=true}) {texture=&T_UI_DataChip_Noise}
  n39 = TextureSample(UVs=Panner(Coordinate=@n26, Time=@n30, Speed=@n37) {bFractionalPart=true}) {texture=&T_UI_DataChip_Noise}
  n40 = TextureSample(UVs=Panner(Coordinate=Add(A=@n35, B=@n26), Time=@n30, Speed=@n37) {bFractionalPart=true}) {texture=&T_UI_DataChip_Noise}
  n41 = Power(
   Base=Multiply(
    A=MaterialFunctionCall(X=@n33.R, Y=@n34.R, Z=@n36.R) {materialFunction=&MakeFloat3}.Result
    B=MaterialFunctionCall(X=@n38.G, Y=@n39.G, Z=@n40.G) {materialFunction=&MakeFloat3}.Result
   )
   Exp=ScalarParameter() {defaultValue=1, parameterName=Color_Power}
  )
  n42 = VectorParameter() {parameterName=Color_01, defaultValue=<rgba 0.6 6 1.23 1>}
  n43 = VectorParameter() {parameterName=Color_02, defaultValue=<rgba 5.839 0 6 1>}
  n44 = Multiply(A=@n25.RGB, B=StaticSwitchParameter(True=@n41, False=LinearInterpolate(A=@n42.RGB, B=@n43.RGB, Alpha=@n41)) {parameterName=UseChromatic?})
  MaterialFunctionCall(Emissive=MaterialFunctionCall(LDR Input=Saturate(None=@n44), HDR Tint=@n44) {materialFunction=&DeriveHDRfromLDR}.RemappedRange) {materialFunction=&MF_FX_EmissiveAdjustments}
  Multiply(A=@n19.R)
  n45 = TextureCoordinate()
  n46 = OneMinus(None=Sine(None=ComponentMask(None=@n45) {g=true, r=true}) {period=2})
  n47 = Saturate(None=Multiply(A=Power(Base=@n46) {constExponent=4}) {constB=3})
  n48 = Add(A=ComponentMask(None=@n47) {r=true}, B=ComponentMask(None=@n47) {g=true})
  n49 = DynamicParameter() {paramNames [InnerEdgePower, TimeSpeed, RandomSeed, ChromaticLerp]}
  n50 = ScalarParameter() {defaultValue=5, parameterName=GridColumns}
  n51 = ScalarParameter() {defaultValue=5, parameterName=GridRows}
  n52 = Floor(None=Multiply(A=TextureCoordinate(), B=AppendVector(A=@n50, B=@n51)))
  n53 = ComponentMask(None=@n52) {g=true}
  n54 = Add(
   A=Round(None=DynamicParameter() {paramNames [InnerEdgePower, TimeSpeed, RandomSeed, ChromaticLerp]}.RandomSeed)
   B=Add(A=Multiply(A=@n53, B=@n50), B=ComponentMask(None=@n52) {r=true})
  )
  n55 = Saturate(None=ComponentMask(None=VectorNoise(World Position=@n54)) {r=true})
  n56 = Step(Y=ScalarParameter() {defaultValue=0.7, parameterName=MaskOutRation}, X=@n55)
  n57 = Multiply(A=Time(), B=ScalarParameter() {defaultValue=0.5, parameterName=Speed}) {constB=0.5}
  n58 = Sine(None=Add(A=@n55, B=Multiply(A=@n55, B=@n57)))
  Multiply(A=@n25.A, B=Multiply(A=Power(Base=@n48, Exp=@n49.InnerEdgePower), B=Add(A=@n56, B=@n58)))
  n59 = Abs(None=@n2)
  Multiply(A=@n59, B=@n59)
  Saturate(None=ComponentMask(None=@n8) {r=true})
  ScalarParameter() {parameterName=OpacityEdge_Power}
  TextureCoordinate()
 ]
}
